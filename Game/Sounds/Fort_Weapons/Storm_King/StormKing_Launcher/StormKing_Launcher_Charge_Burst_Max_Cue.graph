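# Sound cue for the Storm King launcher's maximum-charge burst, using the third-person ranged weapon fire sound class and close projectile attenuation.
# A quality-level switch selects between a branch randomizing three waves with pitch and volume modulation and a lighter branch randomizing two unmodulated waves.
# The other two quality slots are empty.
# Concurrency allows one instance per owner and prevents new plays while one is active.
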SoundCue /Game/Sounds/Fort_Weapons/Storm_King/StormKing_Launcher/StormKing_Launcher_Charge_Burst_Max_Cue
root=QualityLevel_0
nodes [
 {id=Attenuation_0, children [Random_1], properties {attenuationSettings=&Fort_Weapon_Projectile_Close_Athena_Att}}
 {id=Attenuation_9, children [Modulator_15], properties {attenuationSettings=&Fort_Weapon_Projectile_Close_Athena_Att}}
 {id=Modulator_15, children [Random_0], properties {pitchMax=1.1, pitchMin=0.9, volumeMax=1.2, volumeMin=1}}
 {id=QualityLevel_0, children [SoundClass_7,SoundClass_1,null,null]}
 {id=Random_0, children [WavePlayer_0,WavePlayer_1,WavePlayer_2], properties {weights [1,1,1]}}
 {id=Random_1, children [WavePlayer_3,WavePlayer_4], properties {weights [1,1]}}
 {
  id=SoundClass_1
  children [Attenuation_0]
  properties {soundClassOverride=&Fort_WeaponRanged_Fire_3P_PSM}
 }
 {
  id=SoundClass_7
  children [Attenuation_9]
  properties {soundClassOverride=&Fort_WeaponRanged_Fire_3P_PSM}
 }
 {
  id=WavePlayer_0
  properties {soundWaveAssetPtr=&/StormKing_Launcher_Charge_Burst_Max_03}
 }
 {
  id=WavePlayer_1
  properties {soundWaveAssetPtr=&/StormKing_Launcher_Charge_Burst_Max_01}
 }
 {
  id=WavePlayer_2
  properties {soundWaveAssetPtr=&/StormKing_Launcher_Charge_Burst_Max_02}
 }
 {
  id=WavePlayer_3
  properties {soundWaveAssetPtr=&StormKing_Launcher_Charge_Burst_Max_03}
 }
 {
  id=WavePlayer_4
  properties {soundWaveAssetPtr=&StormKing_Launcher_Charge_Burst_Max_01}
 }
]
properties {
 bOverrideConcurrency=true
 maxDistance=4950
 volumeMultiplier=1
 attenuationOverrides {falloffDistance=5000, attenuationShapeExtents {x=1000}}
 concurrencyOverrides {bLimitToOwner=true, maxCount=1, resolutionRule=PreventNew}
}
