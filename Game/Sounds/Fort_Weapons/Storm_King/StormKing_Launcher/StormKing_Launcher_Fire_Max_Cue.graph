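# Sound cue for the Storm King Launcher "Fire Max" weapon sound, mixing close, medium and distant bands, each with its own attenuation and sound class override.
# Close plays a random shot; medium and distant each blend a pitched-down close shot with random NoVox and VoxOnly takes, with the medium vox briefly delayed.
# Modulators vary volume and pitch, and per-owner concurrency stops the oldest instance at the limit.
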SoundCue /Game/Sounds/Fort_Weapons/Storm_King/StormKing_Launcher/StormKing_Launcher_Fire_Max_Cue
root=Mixer_1
nodes [
 {
  id=Attenuation_10
  children [Modulator_14]
  properties {attenuationSettings=&Fort_Weapon_Projectile_Medium_Mono_Athena_Att}
 }
 {
  id=Attenuation_11
  children [Modulator_13]
  properties {attenuationSettings=&Fort_Weapon_Projectile_Distant_Mono_Athena_Att}
 }
 {id=Attenuation_9, children [Modulator_15], properties {attenuationSettings=&Fort_Weapon_Projectile_Close_Athena_Att}}
 {id=Delay_2, children [Random_1], properties {delayMin=0.05}}
 {id=Mixer_1, children [SoundClass_7,SoundClass_6,SoundClass_0], properties {inputVolume [1,0.7,0.6]}}
 {id=Mixer_2, children [Modulator_4,Modulator_2,Modulator_3], properties {inputVolume [0.2,0.4,0.4]}}
 {id=Mixer_5, children [Modulator_5,Random_0,Delay_2], properties {inputVolume [0.2,0.4,0.4]}}
 {id=Modulator_13, children [Mixer_2], properties {pitchMax=1.1, volumeMax=0.8, volumeMin=0.7}}
 {id=Modulator_14, children [Mixer_5], properties {pitchMax=1.1, volumeMax=0.8, volumeMin=0.7}}
 {id=Modulator_15, children [Random_4], properties {volumeMax=1.2, volumeMin=1}}
 {id=Modulator_2, children [Random_3]}
 {id=Modulator_3, children [Random_2]}
 {id=Modulator_4, children [Random_6], properties {pitchMax=0.9, pitchMin=0.8}}
 {id=Modulator_5, children [Random_5], properties {pitchMax=0.9, pitchMin=0.8}}
 {id=Random_0, children [WavePlayer_0,WavePlayer_1,WavePlayer_2,WavePlayer_3,WavePlayer_4], properties {weights [1,1,1,1,1]}}
 {id=Random_1, children [WavePlayer_5,WavePlayer_6,WavePlayer_7,WavePlayer_8,WavePlayer_9], properties {weights [1,1,1,1,1]}}
 {id=Random_2, children [WavePlayer_10, WavePlayer_11, WavePlayer_12, WavePlayer_13, WavePlayer_14], properties {weights [1,1,1,1,1]}}
 {id=Random_3, children [WavePlayer_15, WavePlayer_16, WavePlayer_17, WavePlayer_18, WavePlayer_19], properties {weights [1,1,1,1,1]}}
 {id=Random_4, children [WavePlayer_20, WavePlayer_21, WavePlayer_22], properties {weights [1,1,1]}}
 {id=Random_5, children [WavePlayer_23, WavePlayer_24, WavePlayer_25], properties {weights [1,1,1]}}
 {id=Random_6, children [WavePlayer_26, WavePlayer_57, WavePlayer_58], properties {weights [1,1,1]}}
 {
  id=SoundClass_0
  children [Attenuation_11]
  properties {soundClassOverride=&Fort_WeaponRanged_Fire_Distant}
 }
 {
  id=SoundClass_6
  children [Attenuation_10]
  properties {soundClassOverride=&Fort_WeaponRanged_Fire_Distant}
 }
 {
  id=SoundClass_7
  children [Attenuation_9]
  properties {soundClassOverride=&Fort_WeaponRanged_Fire_3P_PSM}
 }
 {
  id=WavePlayer_0
  properties {soundWaveAssetPtr=&/StormKing_Launcher_Fire_Medium_NoVox_01}
 }
 {
  id=WavePlayer_1
  properties {soundWaveAssetPtr=&/StormKing_Launcher_Fire_Medium_NoVox_02}
 }
 {
  id=WavePlayer_10
  properties {soundWaveAssetPtr=&/StormKing_Launcher_Fire_Distant_VoxOnly_05}
 }
 {
  id=WavePlayer_11
  properties {soundWaveAssetPtr=&/StormKing_Launcher_Fire_Distant_VoxOnly_01}
 }
 {
  id=WavePlayer_12
  properties {soundWaveAssetPtr=&/StormKing_Launcher_Fire_Distant_VoxOnly_02}
 }
 {
  id=WavePlayer_13
  properties {soundWaveAssetPtr=&/StormKing_Launcher_Fire_Distant_VoxOnly_03}
 }
 {
  id=WavePlayer_14
  properties {soundWaveAssetPtr=&/StormKing_Launcher_Fire_Distant_VoxOnly_04}
 }
 {
  id=WavePlayer_15
  properties {soundWaveAssetPtr=&/StormKing_Launcher_Fire_Distant_NoVox_01}
 }
 {
  id=WavePlayer_16
  properties {soundWaveAssetPtr=&/StormKing_Launcher_Fire_Distant_NoVox_02}
 }
 {
  id=WavePlayer_17
  properties {soundWaveAssetPtr=&/StormKing_Launcher_Fire_Distant_NoVox_03}
 }
 {
  id=WavePlayer_18
  properties {soundWaveAssetPtr=&/StormKing_Launcher_Fire_Distant_NoVox_04}
 }
 {
  id=WavePlayer_19
  properties {soundWaveAssetPtr=&/StormKing_Launcher_Fire_Distant_NoVox_05}
 }
 {
  id=WavePlayer_2
  properties {soundWaveAssetPtr=&/StormKing_Launcher_Fire_Medium_NoVox_03}
 }
 {id=WavePlayer_20, properties {soundWaveAssetPtr=&/StormKing_Launcher_Fire_Close_02}}
 {id=WavePlayer_21, properties {soundWaveAssetPtr=&/StormKing_Launcher_Fire_Close_03}}
 {id=WavePlayer_22, properties {soundWaveAssetPtr=&/StormKing_Launcher_Fire_Close_01}}
 {id=WavePlayer_23, properties {soundWaveAssetPtr=&StormKing_Launcher_Fire_Close_02}}
 {id=WavePlayer_24, properties {soundWaveAssetPtr=&StormKing_Launcher_Fire_Close_03}}
 {id=WavePlayer_25, properties {soundWaveAssetPtr=&StormKing_Launcher_Fire_Close_01}}
 {id=WavePlayer_26, properties {soundWaveAssetPtr=&StormKing_Launcher_Fire_Close_02}}
 {
  id=WavePlayer_3
  properties {soundWaveAssetPtr=&/StormKing_Launcher_Fire_Medium_NoVox_04}
 }
 {
  id=WavePlayer_4
  properties {soundWaveAssetPtr=&/StormKing_Launcher_Fire_Medium_NoVox_05}
 }
 {
  id=WavePlayer_5
  properties {soundWaveAssetPtr=&/StormKing_Launcher_Fire_Medium_VoxOnly_01}
 }
 {id=WavePlayer_57, properties {soundWaveAssetPtr=&StormKing_Launcher_Fire_Close_03}}
 {id=WavePlayer_58, properties {soundWaveAssetPtr=&StormKing_Launcher_Fire_Close_01}}
 {
  id=WavePlayer_6
  properties {soundWaveAssetPtr=&/StormKing_Launcher_Fire_Medium_VoxOnly_02}
 }
 {
  id=WavePlayer_7
  properties {soundWaveAssetPtr=&/StormKing_Launcher_Fire_Medium_VoxOnly_03}
 }
 {
  id=WavePlayer_8
  properties {soundWaveAssetPtr=&/StormKing_Launcher_Fire_Medium_VoxOnly_04}
 }
 {
  id=WavePlayer_9
  properties {soundWaveAssetPtr=&/StormKing_Launcher_Fire_Medium_VoxOnly_05}
 }
]
properties {
 bOverrideConcurrency=true
 maxDistance=27500
 volumeMultiplier=0.9
 attenuationOverrides {falloffDistance=5000, attenuationShapeExtents {x=1000}}
 concurrencyOverrides {bLimitToOwner=true, maxCount=3, resolutionRule=StopOldest}
}
